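# Material function giving a world position offset that distorts camera-aligned particle sprites into raindrop shapes, with size X as radius and Y as drop length.
# The drop shape stretches UV-centred corners along particle direction by clamped speed, then passes through a basis from MakeVectorsOrthogonal on direction and camera vectors.
# It blends toward a round view-space billboard at low speed or as squared world-Z of the view direction rises, then subtracts world position.
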MaterialFunction /Engine/Functions/Engine_MaterialFunctions02/WorldPositionOffset/DropletParticleVertexShader
graph {
 properties {
  description=Make a camera aligned particle distort like a raindrop at different viewing angles. The X particle dimension define the sphere radius and the y particle dimension define the water drop length.
  libraryCategoriesText [Misc, world position offset]
 }
 nodes [
  n1 = Subtract(B=CameraPositionWS())
  Add(A=Multiply(A=Normalize(VectorInput=@n1)))
  n2 = ParticleSpeed()
  n3 = FunctionInput() {
   bUsePreviewValueAsDefault=true
   description=Defines how much speed is nessesary to distort the particle systems tail end by one world unit.
   inputName=Speed Per unit of tail distortion
   inputType=FunctionInput_Scalar
   previewValue=<5 0 0 1>
  }
  n4 = FunctionInput() {
   bUsePreviewValueAsDefault=true
   description=Maximum about of tail end distortion.
   inputName=Max tail lengthing percentage
   inputType=FunctionInput_Scalar
   sortPriority=10
   previewValue=<0.1 0 0 1>
  }
  n5 = Multiply(
   A=Clamp(None=Divide(A=@n2, B=@n3) {constB=5, desc="number of world units traveled per world unit of distortion "}, Max=@n4) {maxDefault=10}
   B=ComponentMask(None=ParticleSize()) {g=true}
  )
  n6 = ComponentMask(None=Transform(None=ParticleDirection()) {transformSourceType=TRANSFORMSOURCE_World, transformType=TRANSFORM_Tangent}) {g=true, r=true}
  n7 = Subtract(A=TextureCoordinate()) {constA=0.5, constB=0.5}
  n8 = Multiply(
   A=Multiply(A=DotProduct(A=@n6, B=@n7) {desc=projected gradient along the velocity vector in tangent space}) {constB=0.5, desc=directional growth fudge factor}
   B=@n7
  )
  n9 = Add(A=Multiply(A=Multiply(A=@n5, B=@n6) {constA=1}, B=@n7), B=Multiply(A=ParticleSize(), B=Add(A=@n8, B=@n7)))
  n10 = Constant()
  n11 = ParticleDirection()
  n12 = MaterialFunctionCall() {materialFunction=&CameraDirectionVector}
  n13 = MaterialFunctionCall!(Vector1=CrossProduct(A=@n11, B=@n12.Result), Vector2=@n11, Vector3=@n12.Result) {materialFunction=/Engine/Functions/Engine_MaterialFunctions02/Math/MakeVectorsOrthogonal}
  n14 = ParticlePositionWS()
  n15 = Add(
   A=MaterialFunctionCall(VectorToTransform=AppendVector(A=@n9, B=@n10), BasisX=@n13.Vector1, BasisY=@n13.Vector1, BasisZ=@n13.Vector1) {desc=world orientation, materialFunction=&Transform3x3Matrix}.Transformed Vector
   B=@n14
  ) {desc=world position}
  n16 = Multiply(A=AppendVector(A=@n7, B=Constant()), B=Constant3Vector() {constant=<rgba -1 1 1 0>})
  n17 = ParticleSize()
  n18 = Add(A=@n14, B=Multiply(A=Transform(None=@n16) {transformSourceType=TRANSFORMSOURCE_View}, B=ComponentMask(None=@n17) {r=true}))
  n19 = CameraPositionWS()
  n20 = ComponentMask(None=Normalize(VectorInput=Subtract(A=@n19, B=@n15) {desc=- new world pos})) {b=true}
  n21 = FunctionInput() {
   bUsePreviewValueAsDefault=true
   inputName=Min speed before distortion
   inputType=FunctionInput_Scalar
   sortPriority=10
   previewValue=<10 0 0 1>
  }
  n22 = FunctionInput() {
   bUsePreviewValueAsDefault=true
   description=Velocity divisor
   inputName=Distortion lerp rate
   inputType=FunctionInput_Scalar
   sortPriority=10
   previewValue=<20 0 0 1>
  }
  n23 = Clamp(None=Divide(A=Subtract(A=@n2, B=@n21) {constB=10, desc=min speed before deformation}, B=@n22) {constB=20, desc=speed to become spherical})
  n24 = Clamp(None=Add(A=Multiply(A=@n20, B=@n20) {desc=angle lerp}, B=OneMinus(None=@n23)))
  FunctionOutput(None=Subtract(A=LinearInterpolate(A=@n15, B=@n18, Alpha=@n24), B=WorldPosition().XYZ))
 ]
}
