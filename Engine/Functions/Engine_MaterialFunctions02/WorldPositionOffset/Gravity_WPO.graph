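# Material function, exposed under World Position Offset, that computes the offset and velocity of an object falling under gravity.
# It subtracts Fall Start Time from Current Time, which defaults to material Time ignoring pause, and clamps the elapsed time at its minimum so motion waits for the start.
# Closed-form kinematics yield Offset From Start Position (half g t squared) and Instantaneous Velocity (g t), using an overridable Earth-gravity vector pointing down Z.
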MaterialFunction /Engine/Functions/Engine_MaterialFunctions02/WorldPositionOffset/Gravity_WPO
graph {
 properties {
  bExposeToLibrary=true
  description="Find the velocity and updated position of an object that's falling on earth. "
  libraryCategoriesText [World Position Offset]
 }
 nodes [
  n1 = Time() {bIgnorePause=true}
  n2 = Clamp(None=Subtract(
    A=FunctionInput(Preview=@n1) {bUsePreviewValueAsDefault=true, inputName=Current Time, inputType=FunctionInput_Scalar, sortPriority=1}
    B=FunctionInput() {inputName=Fall Start Time, inputType=FunctionInput_Scalar}
   )) {clampMode=CMODE_ClampMin}
  n3 = FunctionInput() {bUsePreviewValueAsDefault=true, inputName=Gravitational Acceleration, previewValue=<0 0 -980 1>}
  FunctionOutput(None=Multiply(A=Multiply(A=@n2, B=@n2), B=Multiply(A=@n3) {constB=0.5})) {outputName=Offset From Start Position}
  FunctionOutput(None=Multiply(A=@n2, B=@n3)) {outputName=Instantaneous Velocity, sortPriority=1}
 ]
}
