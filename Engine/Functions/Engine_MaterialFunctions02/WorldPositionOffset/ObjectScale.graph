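# ObjectScale is a material function that returns an object's X, Y and Z scale, for use in world position offset and similar material logic.
# It transforms each unit axis vector out of instance space and measures the result with VectorLength to get that axis's scale.
# The three lengths are output separately and also appended into a combined float3.
# Its descriptions state that it is not compatible with the pixel shader.
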MaterialFunction /Engine/Functions/Engine_MaterialFunctions02/WorldPositionOffset/ObjectScale
graph {
 properties {
  description=Returns the object's XYZ scale together and seperately.This is not compatible with the pixel shader.
  libraryCategoriesText [Misc, World position offset]
 }
 nodes [
  n1 = MaterialFunctionCall(Vector 3=Transform(None=Constant3Vector() {constant=<rgba 0 0 1 0>}) {transformSourceType=TRANSFORMSOURCE_Instance}) {materialFunction=/Engine/Functions/Engine_MaterialFunctions02/Utility/VectorLength}
  FunctionOutput(None=@n1.V3 Length) {
   description=Returns a scalar value that is equal to the object's Z-scale. This is not compatible with the pixel shader.
   outputName=Scale Z
   sortPriority=3
  }
  n2 = MaterialFunctionCall(Vector 3=Transform(None=Constant3Vector() {constant=<rgba 1 0 0 0>}) {transformSourceType=TRANSFORMSOURCE_Instance}) {materialFunction=&VectorLength}
  FunctionOutput(None=@n2.V3 Length) {
   description=Returns a scalar value that is equal to the object's X-scale. This is not compatible with the pixel shader.
   outputName=Scale X
   sortPriority=1
  }
  n3 = MaterialFunctionCall(Vector 3=Transform(None=Constant3Vector() {constant=<rgba 0 1 0 0>}) {transformSourceType=TRANSFORMSOURCE_Instance}) {materialFunction=&VectorLength}
  FunctionOutput(None=@n3.V3 Length) {
   description=Returns a scalar value that is equal to the object's Y-scale. This is not compatible with the pixel shader.
   outputName=Scale Y
   sortPriority=2
  }
  FunctionOutput(None=AppendVector(A=AppendVector(A=@n2.V3 Length, B=@n3.V3 Length), B=@n1.V3 Length)) {
   desc="returns 1,1,1 if the objects scale is 1,1,1."
   description=Returns a float3 value that is equal to the objects XYZ scale. This is not compatible with the pixel shader.
   outputName=Scale XYZ
  }
 ]
}
