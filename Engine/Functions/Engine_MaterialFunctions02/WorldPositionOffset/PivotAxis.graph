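# PivotAxis is a World Position Offset material function that creates per-point pivots for RotateAboutAxis, for effects like flag motion.
# One Pivot Axis/Pos input, in instance space, serves as both axis direction and position.
# It projects world position onto the plane through that position perpendicular to the axis, so pivots share one height but vary across it.
# Dividing the transformed axis by object scale supports uniform scaling only; rotation and pivot axes must not coincide.
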MaterialFunction /Engine/Functions/Engine_MaterialFunctions02/WorldPositionOffset/PivotAxis
graph {
 properties {
  bExposeToLibrary=true
  description="Creates a common pivot location on arbitrary axes. This is helpful for creating flag motion. Instead of using a single pivot point near the top of the flag, use a shared z point and unique xy location data to create a more realistic connection along the object's width. Note: this shader node only supports uniform scaling. Also, the rotate about axis axis and the pivot axis/pos should not coincide."
  libraryCategoriesText [Misc, World Position Offset]
 }
 nodes [
  n1 = FunctionInput() {
   bUsePreviewValueAsDefault=true
   description="Enter a number to use as your both your local axis line axis and position. If you would like to lock off your model enter on it's local z axis near its pivot enter (0,0,1). If you would like to lock of the top of the model enter your model's height in as (0,0,mod6el height)."
   inputName=Pivot Axis/Pos
   previewValue=<0 0 1 1>
  }
  n2 = Divide(
   A=Transform(None=Normalize(VectorInput=@n1)) {transformSourceType=TRANSFORMSOURCE_Instance}
   B=MaterialFunctionCall() {materialFunction=&/ObjectScale}.Scale XYZ
  )
  n3 = DotProduct(
   A=@n2
   B=Subtract(
    A=WorldPosition().XYZ
    B=TransformPosition(None=@n1) {transformSourceType=TRANSFORMPOSSOURCE_Instance, transformType=TRANSFORMPOSSOURCE_World}
   )
  )
  FunctionOutput(None=Subtract(A=WorldPosition().XYZ, B=Multiply(A=@n2, B=@n3))) {description=This can be used as the pivot point in the rotate about axis node., outputName=Pivot}
 ]
}
