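# Test material function that generates an animated world-space sine wave plus world normals for it, intended for world position offset.
# World position dotted with a negated XY direction vector, plus time, drives both the sine output and the UVs of a sine-wave normal map.
# The sampled normal gets two-sided sign and XY-flip corrections, then goes through RotateAboutAxis around Z, using the direction's radial angle plus a fudge factor.
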MaterialFunction /Engine/Functions/Engine_MaterialFunctions02/WorldPositionOffset/SineWithNormalSupport
graph {
 properties {
  description="Very specific material function that projects into world space and then maps a normal map in that direction and bends its normals to face in the new direction.  For test purposes only. (May remove eventually)  Outputs worls space normals"
  libraryCategoriesText [Misc, World position offset]
 }
 nodes [
  n1 = FunctionInput(Preview=WorldPosition().XYZ) {bUsePreviewValueAsDefault=true, inputName=World Position, sortPriority=10, previewValue=<-0.004 0 0 1>}
  n2 = FunctionInput() {bUsePreviewValueAsDefault=true, inputName=X world dot, inputType=FunctionInput_Scalar, previewValue=<-0.004 0 0 1>}
  n3 = AppendVector(
   A=AppendVector(A=@n2, B=FunctionInput() {bUsePreviewValueAsDefault=true, inputName=Y world dot, inputType=FunctionInput_Scalar, sortPriority=1})
   B=Constant()
  )
  n4 = FunctionInput(Preview=MaterialFunctionCall(Speed=Constant() {r=0.1}) {materialFunction=&TimeWithSpeedVariable}.Frac Time) {bUsePreviewValueAsDefault=true, inputName=time, inputType=FunctionInput_Scalar, sortPriority=10, previewValue=<-0.004 0 0 1>}
  n5 = Add(A=DotProduct(A=@n1, B=Multiply(A=@n3) {constB=-1}), B=@n4)
  FunctionOutput(None=Sine(None=@n5)) {outputName=SineWave}
  n6 = Normalize(VectorInput=@n3)
  n7 = FunctionInput() {bUsePreviewValueAsDefault=true, inputName=Rotation fudge factor, inputType=FunctionInput_Scalar, previewValue=<0.0867 0 0 1>}
  n8 = Add(
   A=MaterialFunctionCall(Vector or UVs=ComponentMask(None=@n6) {g=true, r=true}) {materialFunction=&VectorToRadialValue}.Vector Converted to Angle
   B=@n7
  ) {constB=0.9}
  n9 = TextureSample(UVs=@n5) {samplerType=SAMPLERTYPE_Normal, texture=/Engine/Functions/Engine_MaterialFunctions02/ExampleContent/Textures/SineWaveWorldNormalMap}
  n10 = Multiply(A=TwoSidedSign(), B=@n9.RGB) {constA=1}
  n11 = AppendVector(A=ComponentMask(None=Multiply(A=@n10) {constB=-1}) {g=true, r=true}, B=@n9.B)
  FunctionOutput(None=RotateAboutAxis(
    NormalizedRotationAxis=Constant3Vector() {constant=<rgba 0 0 1 0>}
    RotationAngle=Frac(None=@n8)
    PivotPoint=Constant3Vector()
    Position=Normalize(VectorInput=@n11)
   )) {outputName=Sine Wave World Normal}
 ]
}
