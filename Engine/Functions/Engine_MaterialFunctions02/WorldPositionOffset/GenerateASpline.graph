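# Material function that evaluates a Catmull-Rom spline segment at t (0-1), exposed to the library for world position offset use.
# Two control points and two tangent handles feed an expanded cubic polynomial for position and its analytic derivative for raw and normalized tangents.
# Chained cross products of the tangent with a reference vector, defaulting to object orientation, build normal and bi-normal outputs for an orthonormal frame along the curve.
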MaterialFunction /Engine/Functions/Engine_MaterialFunctions02/WorldPositionOffset/GenerateASpline
graph {
 properties {bExposeToLibrary=true, description=Generate a Catmull-Rom spline., libraryCategoriesText [Misc, World Position Offset]}
 nodes [
  n1 = FunctionInput() {bUsePreviewValueAsDefault=true, description=Control point 1, inputName=P1, sortPriority=2}
  n2 = FunctionInput() {bUsePreviewValueAsDefault=true, description=Tangent handle 1, inputName=P0, sortPriority=1}
  n3 = Multiply(A=@n2) {constB=-1}
  n4 = FunctionInput() {bUsePreviewValueAsDefault=true, description=Control point 2, inputName=P2, sortPriority=3}
  n5 = FunctionInput() {bUsePreviewValueAsDefault=true, description="Spline sample position. (0-1)", inputName=t, inputType=FunctionInput_Scalar}
  n6 = Add(A=Multiply(A=@n1) {constB=2}, B=Multiply(A=Add(A=@n3, B=@n4), B=@n5))
  n7 = Subtract(A=Multiply(A=@n2) {constB=2}, B=Multiply(A=@n1) {constB=5})
  n8 = FunctionInput() {bUsePreviewValueAsDefault=true, description=Tangent Handle 2, inputName=P3, sortPriority=4}
  n9 = Subtract(A=Multiply(A=@n4) {constB=4}, B=@n8)
  n10 = Subtract(A=Add(A=Multiply(A=@n2) {constB=-1}, B=Multiply(A=@n1) {constB=3}), B=Multiply(A=@n4) {constB=3})
  n11 = Add(A=Multiply(A=Add(A=@n7, B=@n9), B=Power(Base=@n5)), B=Multiply(A=Add(A=@n10, B=@n8), B=Power(Base=@n5) {constExponent=3}))
  FunctionOutput(None=Multiply(A=Add(A=@n6, B=@n11)) {constB=0.5}) {description=A world position sampled at t., outputName=T position}
  n12 = Add(A=Multiply(A=@n2) {constB=-1}, B=@n4)
  n13 = Add(A=Subtract(A=Multiply(A=@n2) {constB=2}, B=Multiply(A=@n1) {constB=5}), B=Multiply(A=@n4) {constB=4})
  n14 = Multiply(A=Multiply(A=Subtract(A=@n13, B=@n8), B=@n5)) {constB=2}
  n15 = Add(A=Multiply(A=@n2) {constB=-1}, B=Multiply(A=@n1) {constB=3})
  n16 = Multiply(A=@n4) {constB=3}
  n17 = Multiply(A=Add(A=Subtract(A=@n15, B=@n16), B=@n8), B=Power(Base=@n5))
  n18 = Multiply(A=Add(A=Add(A=@n12, B=@n14), B=Multiply(A=@n17) {constB=3})) {constB=0.5}
  n19 = Normalize(VectorInput=@n18)
  FunctionOutput(None=@n19) {description=Vector down the length of the spline., outputName=T Tangent}
  FunctionOutput(None=@n18) {outputName=UnnormalizedTangent, sortPriority=-1}
  n20 = FunctionInput(Preview=ObjectOrientation()) {
   bUsePreviewValueAsDefault=true
   description=The default value is object orientation.
   inputName=Normal/Bi-Normal Cross product vector
   sortPriority=5
  }
  n21 = Normalize(VectorInput=CrossProduct(A=CrossProduct(A=@n19, B=@n20), B=@n19))
  FunctionOutput(None=Normalize(VectorInput=CrossProduct(A=@n21, B=@n19))) {outputName=T Bi-normal}
  FunctionOutput(None=@n21) {outputName=T Normal}
 ]
}
