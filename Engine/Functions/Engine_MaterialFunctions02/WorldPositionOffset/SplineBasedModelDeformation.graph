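# Material function that bends a mesh along a four-point spline through World Position Offset, with P0 and P3 as tangent handles.
# Instance-space position projected on the deformation vector, divided by model height, gives t for GenerateASpline.
# The cross-section's X/Y offsets are rebuilt on the spline's binormal and normal, scaled by Thickness.
# A height input displaces along the resulting cylindrical normal; outputs also include pixel normals, spline frame and debug gradient.
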MaterialFunction /Engine/Functions/Engine_MaterialFunctions02/WorldPositionOffset/SplineBasedModelDeformation
graph {
 properties {
  bExposeToLibrary=true
  description=Deform a model around a spline that is created with the four control point inputs. P0 through P3. P0 and P3 are tangent handles.
  libraryCategoriesText [Misc, World Position Offset]
 }
 nodes [
  n1 = FunctionInput() {
   bUsePreviewValueAsDefault=true
   description=Local axis that the model will use to map to the length of the spline.
   inputName=LocalDeformationVector
   sortPriority=5
   previewValue=<0 0 1 1>
  }
  n2 = Normalize(VectorInput=@n1)
  n3 = FunctionInput(Preview=TransformPosition(None=WorldPosition() {worldPositionShaderOffset=WPT_ExcludeAllShaderOffsets}.XYZ) {transformSourceType=TRANSFORMPOSSOURCE_World, transformType=TRANSFORMPOSSOURCE_Instance}) {bUsePreviewValueAsDefault=true, description=Automatically supplied., inputName=Local Position, sortPriority=10, previewValue=<0 0 1 1>}
  n4 = DotProduct(A=@n2, B=@n3)
  n5 = FunctionInput() {
   bUsePreviewValueAsDefault=true
   description="This is the models local space height along the spline tangent. If a mesh is being deformed along it's z axis and the distance from the pivot point of the model to the top of the mesh, in local space, is 1 then you would enter 1 in this field. If the model's height were 10 then 10 should be entered."
   inputName=Model height along deformation vector
   inputType=FunctionInput_Scalar
   sortPriority=5
   previewValue=<1 0 0 1>
  }
  n6 = Clamp(None=Divide(A=@n4, B=@n5)) {desc=0-1 along deformation vector}
  n7 = MaterialFunctionCall!(
   t=@n6
   P0=FunctionInput() {bUsePreviewValueAsDefault=true, description="Control point. 3D world space position. ", inputName=P0}
   P1=FunctionInput() {bUsePreviewValueAsDefault=true, description="Control point. 3D world space position. ", inputName=P1}
   P2=FunctionInput() {bUsePreviewValueAsDefault=true, description="Control point. 3D world space position. ", inputName=P2}
   P3=FunctionInput() {bUsePreviewValueAsDefault=true, description="Control point. 3D world space position. ", inputName=P3}
   Normal/Bi-Normal Cross product vector
  ) {desc=World Position Offset, materialFunction=&/GenerateASpline}
  n8 = Subtract(A=@n3, B=Multiply(A=@n2, B=@n4) {desc=Center line local space}) {desc=X Y Vector around center line}
  n9 = Multiply(A=@n7.T Bi-normal, B=DotProduct(A=@n8, B=Constant3Vector() {constant=<rgba 1 0 0 0>}))
  n10 = Multiply(A=DotProduct(A=@n8, B=Constant3Vector() {constant=<rgba 0 1 0 0>}), B=@n7.T Normal)
  n11 = Add(A=@n9, B=@n10)
  n12 = FunctionInput(Preview=@n11) {
   bUsePreviewValueAsDefault=true
   description=Automatically retains the model's intial scale. If desired enter a new value to scale the spline deformed models x and y local axes positions.
   inputName=Thickness
   inputType=FunctionInput_Vector2
   previewValue=<20 20 0 1>
  }
  n13 = MaterialFunctionCall(Float2=@n12) {materialFunction=&BreakOutFloat2Components}
  n14 = Add(
   A=Add(A=Multiply(A=@n13.R, B=@n9), B=Multiply(A=@n13.G, B=@n10))
   B=TransformPosition(None=@n7.T position) {transformSourceType=TRANSFORMPOSSOURCE_World, transformType=TRANSFORMPOSSOURCE_Instance}
  ) {desc=new local position}
  n15 = Subtract(A=@n14, B=Multiply(A=@n4, B=@n2))
  n16 = MaterialFunctionCall(Vector=@n11) {materialFunction=&SafeNormalize}
  n17 = FunctionInput() {
   bUsePreviewValueAsDefault=true
   description=The intensity will determine amount of offset.
   inputName=Optional Greyscale Height Map
   inputType=FunctionInput_Scalar
   sortPriority=101
   previewValue=<0 0 1 1>
  }
  FunctionOutput(None=Add(
    A=TransformPosition(None=@n15) {transformSourceType=TRANSFORMPOSSOURCE_Instance, transformType=TRANSFORMPOSSOURCE_World}
    B=Multiply(A=@n16.Result, B=@n17)
   )) {outputName=World Position Offset}
  FunctionOutput(None=@n16.Result) {description=In world and does not include transform the original normals, outputName="Pixel normal (WS Cylindrical)", sortPriority=10}
  FunctionOutput(None=@n7.T Bi-normal) {description=local x axis, outputName="BiNormal (X)", sortPriority=5}
  n18 = FunctionInput() {bUsePreviewValueAsDefault=true, inputName=Optional Tangent Space Normal Map, sortPriority=100, previewValue=<0 0 1 1>}
  n19 = MaterialFunctionCall!(Vector1=Transform(None=@n7.T Tangent) {desc=tangent z, transformSourceType=TRANSFORMSOURCE_Instance}, Vector2=@n16.Result) {desc=cyl normal, materialFunction=/Engine/Functions/Engine_MaterialFunctions02/Math/CreateThirdOrthogonalVector}
  FunctionOutput(None=MaterialFunctionCall(VectorToTransform=@n18, BasisX=@n19.Vector1, BasisY=@n19.Vector1, BasisZ=@n19.Vector1) {materialFunction=&Transform3x3Matrix}.Transformed Vector) {description=In world and does not include transform the original normals, outputName="Pixel normal (WS Texture)", sortPriority=11}
  FunctionOutput(None=@n7.T Normal) {description=local Y axis, outputName="Normal (Y)", sortPriority=6}
  FunctionOutput(None=@n7.T Tangent) {description=local z, outputName="Tangent (Z)", sortPriority=7}
  FunctionOutput(None=@n7.T position) {outputName=World T Position, sortPriority=8}
  FunctionOutput(None=@n6) {
   description="This returns the spline length values being used to drive the spline deformation (t). The values should stat at 0 (at the pivot location) and gradually become 1 near the furthest extent of the object from the pivot point in the direction of the deformation vector."
   outputName=Debug Gradient
   sortPriority=1000
  }
  FunctionOutput(None=@n14) {outputName=Local Position Offset, sortPriority=1}
 ]
}
