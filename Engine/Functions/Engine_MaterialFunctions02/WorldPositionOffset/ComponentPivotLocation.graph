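# ComponentPivotLocation is a library-exposed material function that returns the world position of the component's origin.
# It feeds a zero constant vector through a TransformPosition node from instance space to world space and outputs the result.
# Because the source is instance space, the pivot follows each instance's transform where instancing is used.
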MaterialFunction /Engine/Functions/Engine_MaterialFunctions02/WorldPositionOffset/ComponentPivotLocation
graph {
 properties {bExposeToLibrary=true, description=Returns the world position of the components origin.}
 nodes [FunctionOutput(None=TransformPosition(None=Constant3Vector()) {transformSourceType=TRANSFORMPOSSOURCE_Instance, transformType=TRANSFORMPOSSOURCE_World})]
}
