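# Material function that aligns a static mesh to the camera through World Position Offset, with custom basis vectors to rotate it.
# Instance-space position, multiplied by object scale, passes through an orthogonalized basis, then from view to world space around the pivot.
# Normals get the same rotation, and extra outputs expose a basis vector and the pixel-shader world position.
# A separate "test WPO" output inverse-transforms the result using a camera-vector basis.
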MaterialFunction /Engine/Functions/Engine_MaterialFunctions02/WorldPositionOffset/AlignMeshToTheCamera
graph {
 properties {
  bExposeToLibrary=true
  description=Aligns a static mesh to the camera and allows the user to rotate the mesh by altering the objects basis vectors.
  libraryCategoriesText [Misc, World Position Offset]
 }
 nodes [
  n1 = FunctionInput(Preview=WorldPosition() {worldPositionShaderOffset=WPT_ExcludeAllShaderOffsets}.XYZ) {bUsePreviewValueAsDefault=true, inputName=WorldPosition}
  n2 = FunctionInput() {bUsePreviewValueAsDefault=true, inputName=Custom Object Basis 1, sortPriority=2, previewValue=<0 0 1 0>}
  n3 = FunctionInput() {bUsePreviewValueAsDefault=true, inputName=Custom Object Basis 2, sortPriority=3, previewValue=<0 1 0 0>}
  n4 = FunctionInput() {bUsePreviewValueAsDefault=true, inputName=Custom Object Basis 3, sortPriority=4, previewValue=<1 0 0 0>}
  n5 = MaterialFunctionCall!(
   Vector1=Transform(None=@n2) {transformSourceType=TRANSFORMSOURCE_World, transformType=TRANSFORM_Instance}
   Vector2=Transform(None=@n3) {transformSourceType=TRANSFORMSOURCE_World, transformType=TRANSFORM_Instance}
   Vector3=Transform(None=@n4) {transformSourceType=TRANSFORMSOURCE_World, transformType=TRANSFORM_Instance}
  ) {materialFunction=/Engine/Functions/Engine_MaterialFunctions02/Math/MakeVectorsOrthogonal}
  n6 = MaterialFunctionCall(
   VectorToTransform=Multiply(
    A=MaterialFunctionCall() {materialFunction=&/ObjectScale}.Scale XYZ
    B=TransformPosition(None=@n1) {transformSourceType=TRANSFORMPOSSOURCE_World, transformType=TRANSFORMPOSSOURCE_Instance}
   ) {desc=Apply Object Scale once again}
   BasisX=@n5.Vector1
   BasisY=@n5.Vector1
   BasisZ=@n5.Vector1
  ) {materialFunction=&Transform3x3Matrix}
  n7 = Transform(None=@n6.Transformed Vector) {transformSourceType=TRANSFORMSOURCE_View}
  n8 = WorldPosition()
  n9 = FunctionInput(Preview=MaterialFunctionCall() {materialFunction=&ComponentPivotLocation}.Result) {bUsePreviewValueAsDefault=true, inputName=PivotLocation}
  FunctionOutput(None=Add(A=Subtract(A=@n7, B=@n8.XYZ), B=@n9)) {outputName=World Position Offset, sortPriority=1}
  n10 = MaterialFunctionCall!(Use World Position Offsets) {materialFunction=/Engine/Functions/Engine_MaterialFunctions02/Utility/CameraVectorWithWPOOptions}
  n11 = CrossProduct(A=@n10.Result, B=Transform(None=Constant3Vector() {constant=<rgba 0 1 0 0>}) {transformSourceType=TRANSFORMSOURCE_View})
  n12 = MaterialFunctionCall(Vector1=@n10.Result, Vector2=@n11, Vector3=CrossProduct(A=@n11, B=@n10.Result)) {materialFunction=&MakeVectorsOrthogonal}
  n13 = MaterialFunctionCall!(VectorToTransform=@n6.Transformed Vector, BasisX=@n12.Vector1, BasisY=@n12.Vector1, BasisZ=@n12.Vector1) {materialFunction=/Engine/Functions/Engine_MaterialFunctions02/Math/InverseTransformMatrix}
  FunctionOutput(None=Add(A=Subtract(A=@n8.XYZ, B=@n13.Transformed Vector 3), B=@n9)) {outputName=test WPO, sortPriority=1}
  n14 = FunctionInput(Preview=VertexNormalWS()) {bUsePreviewValueAsDefault=true, description="Enter the models world space normals. ", inputName=WS Normals}
  FunctionOutput(None=Transform(None=MaterialFunctionCall(VectorToTransform=@n14, BasisX=@n5.Vector1, BasisY=@n5.Vector1, BasisZ=@n5.Vector1) {materialFunction=&Transform3x3Matrix}.Transformed Vector) {transformSourceType=TRANSFORMSOURCE_View}) {outputName=Rotated Normals}
  FunctionOutput(None=@n5.Vector1) {outputName=CameraXVector, sortPriority=1}
  FunctionOutput(None=Add(A=@n7, B=@n9)) {outputName=Undeformed Pixel Shader World Position, sortPriority=2}
 ]
}
