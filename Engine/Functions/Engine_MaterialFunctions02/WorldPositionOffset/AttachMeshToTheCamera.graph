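# Material function that outputs a World Position Offset attaching a mesh to the camera, with adjustable object bases and camera-space position offset.
# It converts world position to instance space, reapplies object scale, then transforms it by a 3x3 basis and the offset.
# The bases pass through MakeVectorsOrthogonal, though all three axis inputs read its Vector1 output.
# The result goes from view to world space, adds camera position and subtracts world position.
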MaterialFunction /Engine/Functions/Engine_MaterialFunctions02/WorldPositionOffset/AttachMeshToTheCamera
graph {
 properties {
  bExposeToLibrary=true
  description=Attaches a static mesh to the camera and allows the user to rotate the mesh by altering the object bases and also provides a camera space offset variable.
  libraryCategoriesText [Misc, World Position Offset]
 }
 nodes [
  n1 = MaterialFunctionCall() {materialFunction=&/ObjectScale}
  n2 = TransformPosition(None=WorldPosition().XYZ) {transformSourceType=TRANSFORMPOSSOURCE_World, transformType=TRANSFORMPOSSOURCE_Instance}
  n3 = MaterialFunctionCall!(
   Vector1=FunctionInput() {bUsePreviewValueAsDefault=true, inputName=Object Basis 1, previewValue=<1 0 0 0>}
   Vector2=FunctionInput() {bUsePreviewValueAsDefault=true, inputName=Object Basis 3, previewValue=<0 0 -1 1>}
   Vector3=FunctionInput() {bUsePreviewValueAsDefault=true, inputName=Object Basis 2, previewValue=<0 1 0 1>}
  ) {materialFunction=/Engine/Functions/Engine_MaterialFunctions02/Math/MakeVectorsOrthogonal}
  n4 = FunctionInput(Preview=Constant3Vector() {constant=<rgba 0 0 200 0>}) {bUsePreviewValueAsDefault=true, inputName=Camera Space Position Offset, sortPriority=3, previewValue=<0 0 200 1>}
  n5 = Transform(None=MaterialFunctionCall(
    VectorToTransform=Multiply(A=@n1.Scale XYZ, B=@n2) {desc=Apply Object Scale once again}
    BasisX=@n3.Vector1
    BasisY=@n3.Vector1
    BasisZ=@n3.Vector1
    BasisPosition=@n4
   ) {materialFunction=&Transform3x3Matrix}.Transformed Vector) {transformSourceType=TRANSFORMSOURCE_View}
  n6 = CameraPositionWS()
  FunctionOutput(None=Subtract(A=Add(A=@n5, B=@n6), B=WorldPosition().XYZ)) {outputName=World Position Offset}
 ]
}
