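# CameraOffset is a library material function that produces a World Position Offset moving geometry toward or away from the camera, so it appears in front of or behind other objects.
# It normalizes the vector from the input position (world position by default) to the camera and scales it by Offset Amount.
# A second output caps the amount at camera distance minus Clamp Padding, limiting only the maximum to avoid intersecting the camera.
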MaterialFunction /Engine/Functions/Engine_MaterialFunctions02/WorldPositionOffset/CameraOffset
graph {
 properties {
  bExposeToLibrary=true
  description="Offset objects in the direction of the camera to either make them appear in front or behind other objects. "
  libraryCategoriesText [Misc, World Position Offset]
 }
 nodes [
  n1 = FunctionInput(Preview=WorldPosition().XYZ) {
   bUsePreviewValueAsDefault=true
   description="Enter the model's vertex world position. Default Value = World Position. (Note: this used to default to actor position but world position is more useful and mimics the particle effects camera offset.)"
   inputName=World Position
   sortPriority=3
  }
  n2 = Subtract(A=CameraPositionWS(), B=@n1)
  n3 = Normalize(VectorInput=@n2)
  n4 = FunctionInput() {
   bUsePreviewValueAsDefault=true
   description="Enter a negative or positive number to offset your model in the direction of the camera. Note that positive values will bring the model closer to the camera and cause rendering errors if the mesh greatly exceeds the models bounding box. "
   inputName=Offset Amount
   inputType=FunctionInput_Scalar
  }
  FunctionOutput(None=Multiply(A=@n3, B=@n4)) {
   description=Either add this value to other world position offset code or enter it directily intro the world position offset master material nub to offset your mesh.
   outputName=Camera Offset
   sortPriority=1
  }
  n5 = FunctionInput() {
   bUsePreviewValueAsDefault=true
   description="Amount of padding to prevent the clamped version of the offset value from clipping into the camera. "
   inputName=Clamp Padding
   inputType=FunctionInput_Scalar
   sortPriority=2
   previewValue=<20 0 0 1>
  }
  n6 = Subtract(A=Distance(A=@n2, B=Constant()), B=@n5)
  FunctionOutput(None=Multiply(A=@n3, B=Clamp(None=@n4, Max=@n6) {clampMode=CMODE_ClampMax})) {
   description="Camera offset clamped to avoid early camera intersections. Adjust the padding to change the amount of space reserved to avoid camera intersections. "
   outputName=Clamped Camera Offset
  }
 ]
}
