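# Material function that outputs the local X, Y and Z basis vectors of a component in world space, filed under the World Position Offset functions.
# Three constant unit axis vectors each pass through a vector Transform into world space, feeding the X, Y and Z outputs.
# The transform source is Instance space rather than Local, so the axes follow per-instance transforms.
# It takes no inputs and does not normalize the results.
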MaterialFunction /Engine/Functions/Engine_MaterialFunctions02/WorldPositionOffset/ComponentBasisVectors
graph {
 properties {description=The local XYZ component transform vectors in world space.}
 nodes [
  FunctionOutput(None=Transform(None=Constant3Vector() {constant=<rgba 0 1 0 0>}) {transformSourceType=TRANSFORMSOURCE_Instance}) {description=The local Y vector transformed into world space., outputName=Y, sortPriority=1}
  FunctionOutput(None=Transform(None=Constant3Vector() {constant=<rgba 0 0 1 0>}) {transformSourceType=TRANSFORMSOURCE_Instance}) {description=The local Z vector transformed into world space., outputName=Z, sortPriority=2}
  FunctionOutput(None=Transform(None=Constant3Vector() {constant=<rgba 1 0 0 0>}) {transformSourceType=TRANSFORMSOURCE_Instance}) {description=The local X vector transformed into world space., outputName=X}
 ]
}
