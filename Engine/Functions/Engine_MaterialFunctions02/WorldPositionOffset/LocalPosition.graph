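# LocalPosition is a deprecated engine material function that exposes local position variants, marked for replacement by the built-in LocalPosition node.
# It has no inputs and wraps four LocalPosition nodes, each feeding its XYZ to an output.
# The outputs cover primitive-relative and instance-relative positions, each available normally or with offsets excluded.
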MaterialFunction /Engine/Functions/Engine_MaterialFunctions02/WorldPositionOffset/LocalPosition
graph {
 properties {
  description=Replace this material function with the built-in LocalPosition node
  userExposedCaption="LocalPosition (deprecated)"
  libraryCategoriesText [Misc,Coordinates]
 }
 nodes [
  FunctionOutput(None=LocalPosition() {includedOffsets=ExcludeOffsets}.XYZ) {outputName="Instance Local Position (Excluding Offsets)", sortPriority=3}
  FunctionOutput(None=LocalPosition().XYZ) {outputName=Instance Local Position, sortPriority=1}
  FunctionOutput(None=LocalPosition() {localOrigin=Primitive}.XYZ) {outputName=Local Position}
  FunctionOutput(None=LocalPosition() {includedOffsets=ExcludeOffsets, localOrigin=Primitive}.XYZ) {outputName="Local Position (Excluding Offsets)", sortPriority=2}
 ]
}
