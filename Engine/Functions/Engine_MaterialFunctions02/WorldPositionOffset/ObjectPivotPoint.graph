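# ObjectPivotPoint is a library material function that returns an object's pivot point in world space, intended for vertex-shader use such as world position offset.
# A zero vector feeds two TransformPosition nodes that convert the origin to world space.
# The default local-space transform gives Object Pivot Location, while the instance-space transform gives Mesh Particle Pivot Location.
# Its descriptions state that it is incompatible with the pixel shader.
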MaterialFunction /Engine/Functions/Engine_MaterialFunctions02/WorldPositionOffset/ObjectPivotPoint
graph {
 properties {
  bExposeToLibrary=true
  description=Returns the object's pivot point in world space. This is not compatible with the pixel shader.
  libraryCategoriesText [Misc, world position offset]
 }
 nodes [
  n1 = Constant3Vector()
  FunctionOutput(None=TransformPosition(None=@n1) {transformSourceType=TRANSFORMPOSSOURCE_Instance, transformType=TRANSFORMPOSSOURCE_World}) {
   description=Returns the object's pivot point in world space. This only works with the vertex shader.
   outputName=Mesh Particle Pivot Location
   sortPriority=1
  }
  FunctionOutput(None=TransformPosition(None=@n1) {transformType=TRANSFORMPOSSOURCE_World}) {description=Returns the object's pivot point in world space. This only works with the vertex shader., outputName=Object Pivot Location}
 ]
}
