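# Material function that cheaply rotates vertices around world X, Y or Z axes, for use as World Position Offset.
# Rotation Amount (1 = full turn) drives sine and cosine applied to WorldPosition minus PivotPoint, which defaults to the object pivot.
# Each axis output rotates two components, then subtracts the original relative position to yield an offset.
# The Y-Axis output swaps X and Z, so its offset is generally nonzero at zero rotation.
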MaterialFunction /Engine/Functions/Engine_MaterialFunctions02/WorldPositionOffset/RotateAboutWorldAxis_cheap
graph {
 properties {
  bExposeToLibrary=true
  description=This function cheaply rotates objects around world axes. Input the angle you would like to use and attach the output to world position offset.
  libraryCategoriesText [Misc, World Position Offset]
 }
 nodes [
  n1 = FunctionInput() {
   bUsePreviewValueAsDefault=true
   description=A value of one is equal to one full rotation.
   inputName=Rotation Amount
   inputType=FunctionInput_Scalar
  }
  n2 = ComponentMask(None=@n1) {r=true}
  n3 = Add(A=Constant(), B=@n2) {desc=theta}
  n4 = Cosine(None=@n3)
  n5 = FunctionInput(Preview=WorldPosition().XYZ) {
   bUsePreviewValueAsDefault=true
   description=World space location of individual vertices. The WorldPosition node is typically used.
   inputName=WorldPosition
   sortPriority=3
  }
  n6 = FunctionInput(Preview=MaterialFunctionCall() {materialFunction=&ObjectPivotPoint}.Object Pivot Location) {
   bUsePreviewValueAsDefault=true
   description=World space location of pivot point to rotate around. The default value is the model's pivot point.
   inputName=PivotPoint
   sortPriority=1
  }
  n7 = Subtract(A=@n5, B=@n6)
  n8 = Add(A=@n7, B=Constant()) {desc=pos}
  n9 = ComponentMask(None=@n8) {r=true}
  n10 = ComponentMask(None=@n8) {g=true}
  n11 = Sine(None=@n3)
  n12 = Sine(None=@n3)
  n13 = ComponentMask(None=@n8) {r=true}
  n14 = ComponentMask(None=@n8) {g=true}
  n15 = Cosine(None=@n3)
  n16 = AppendVector(A=Subtract(A=Multiply(A=@n4, B=@n9), B=Multiply(A=@n10, B=@n11)), B=Add(A=Multiply(A=@n12, B=@n13), B=Multiply(A=@n14, B=@n15)))
  n17 = ComponentMask(None=@n8) {b=true}
  FunctionOutput(None=Subtract(A=AppendVector(A=@n16, B=@n17), B=@n8)) {description="Cheaply rotates the object around the world Z-axis. ", outputName=Z-Axis, sortPriority=1}
  n18 = Add(A=Constant(), B=@n2) {desc=theta}
  n19 = Cosine(None=@n18)
  n20 = Add(A=@n7, B=Constant()) {desc=pos}
  n21 = ComponentMask(None=@n20) {b=true}
  n22 = ComponentMask(None=@n20) {r=true}
  n23 = Sine(None=@n18)
  n24 = AppendVector(A=Subtract(A=Multiply(A=@n19, B=@n21), B=Multiply(A=@n22, B=@n23)), B=ComponentMask(None=@n20) {g=true})
  n25 = Add(A=Multiply(A=Sine(None=@n18), B=ComponentMask(None=@n20) {b=true}), B=Multiply(A=ComponentMask(None=@n20) {r=true}, B=Cosine(None=@n18)))
  FunctionOutput(None=Subtract(A=AppendVector(A=@n24, B=@n25), B=@n20)) {description="Cheaply rotates the object around the world Y-axis. ", outputName=Y-Axis, sortPriority=1}
  n26 = Add(A=@n7, B=Constant()) {desc=pos}
  n27 = Add(A=Constant(), B=@n2) {desc=theta}
  n28 = Cosine(None=@n27)
  n29 = ComponentMask(None=@n26) {g=true}
  n30 = Sine(None=@n27)
  n31 = ComponentMask(None=@n26) {b=true}
  n32 = AppendVector(A=ComponentMask(None=@n26) {r=true}, B=Subtract(A=Multiply(A=@n28, B=@n29), B=Multiply(A=@n30, B=@n31)))
  n33 = Add(A=Multiply(A=Sine(None=@n27), B=ComponentMask(None=@n26) {g=true}), B=Multiply(A=Cosine(None=@n27), B=ComponentMask(None=@n26) {b=true}))
  FunctionOutput(None=Subtract(A=AppendVector(A=@n32, B=@n33), B=@n26)) {description="Cheaply rotates the object around the world X-axis. ", outputName=X-Axis}
 ]
}
