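# Material function that turns a particle sprite or plane into a velocity-aligned pseudo-3D capsule.
# Particle speed sets the default stretch, which drives a World Position Offset building the card from direction and camera vectors, with optional camera pull.
# StretchGradient distorts UVs by stretch and view alignment so sphere textures supply capsule opacity, world-space normals and Pixel Depth Offset.
# It also outputs the UVs, scale factor and a front-face mask.
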MaterialFunction /Engine/Functions/Engine_MaterialFunctions02/WorldPositionOffset/Sprite_Capsule
graph {
 properties {
  bExposeToLibrary=true
  description="The material function performs a number of functions to turn a particle sprite, or plane, into a velocity aligned capsule. "
  libraryCategoriesText [World Position Offset]
 }
 nodes [
  FunctionInput() {
   bUsePreviewValueAsDefault=true
   inputName=---------Scale Over Velocity Parameters---------
   inputType=FunctionInput_StaticBool
   sortPriority=10
  }
  FunctionInput() {bUsePreviewValueAsDefault=true, inputName=---------- Advanced -----------, sortPriority=100}
  n1 = FunctionInput() {
   bUsePreviewValueAsDefault=true
   description="1 will align the leading edge of the capsule to the particle position, -1 will align the trailing edge."
   inputName=Capsule Pivot Alignment
   inputType=FunctionInput_Scalar
  }
  n2 = ParticleSize()
  n3 = ComponentMask(None=@n2) {g=true}
  n4 = FunctionInput(Preview=@n3) {bUsePreviewValueAsDefault=true, inputName=Maximum Scale Factor, inputType=FunctionInput_Scalar, sortPriority=15, previewValue=<2 0 0 1>}
  n5 = ParticleSpeed()
  n6 = FunctionInput() {
   bUsePreviewValueAsDefault=true
   description=The particle must go this fast before beginning to warp along the velocity vector.
   inputName=Minimum Speed
   inputType=FunctionInput_Scalar
   sortPriority=12
  }
  n7 = FunctionInput(Preview=@n3) {
   bUsePreviewValueAsDefault=true
   description="After reaching the minimum speed, the particle will start to warp until this speed is reached (in addition to the minimum speed). This value uses the sprites size's y coordinate by default."
   inputName=Max Speed
   inputType=FunctionInput_Scalar
   sortPriority=13
   previewValue=<30 0 0 1>
  }
  n8 = Divide(
   A=Subtract(
    A=FunctionInput(Preview=@n5) {bUsePreviewValueAsDefault=true, inputName=Particle Speed, inputType=FunctionInput_Scalar, sortPriority=11}
    B=@n6
   )
   B=@n7
  )
  n9 = FunctionInput(Preview=LinearInterpolate(
    A=FunctionInput() {bUsePreviewValueAsDefault=true, inputName=Minimum Scale Factor, inputType=FunctionInput_Scalar, sortPriority=14}
    B=@n4
    Alpha=Clamp(None=@n8)
   )) {
   bUsePreviewValueAsDefault=true
   description="Note: By default, an equation using the \"Scale Over Velocity Parameters\" sets this value. Plugging an input into this field will allow one to manually control the scale factor. "
   inputName=Scale Along Velocity Vector
   inputType=FunctionInput_Scalar
   sortPriority=9
  }
  n10 = Max(A=Subtract(A=@n9, B=NamedRerouteUsage())) {constB=0.001}
  n11 = Multiply(A=Multiply(A=@n1, B=Multiply(A=@n10) {constB=-0.5, desc=Scale}), B=NamedRerouteUsage()) {desc=direction}
  n12 = CrossProduct(A=NamedRerouteUsage(), B=NamedRerouteUsage())
  n13 = Multiply(A=@n12, B=NamedRerouteUsage()) {desc=length}
  n14 = MaterialFunctionCall(Float2=Subtract(A=TextureCoordinate(), B=Constant2Vector() {g=0.5, r=0.5}) {constB=0.5}) {materialFunction=&BreakOutFloat2Components}
  n15 = Multiply(A=@n14.G, B=NamedRerouteUsage()) {desc=Width}
  n16 = NamedRerouteUsage()
  n17 = NamedRerouteUsage()
  n18 = Abs(None=NamedRerouteUsage())
  n19 = Constant() {r=0.0036}
  n20 = Subtract(A=@n18, B=Subtract(A=Constant() {r=1}, B=@n19))
  n21 = Divide(B=@n19)
  n22 = LinearInterpolate(A=@n14.R, B=Multiply(A=Sign(None=@n17)) {constB=-0.5}, Alpha=Clamp(None=Multiply(A=@n20, B=@n21))) {constB=-0.5}
  n23 = FunctionInput(Preview=StaticBool() {value=true}) {
   bUsePreviewValueAsDefault=true
   description=This option will better recreates a capsule with a slight increase in cost. It's enabled by default.
   inputName=Use Advanced Features
   inputType=FunctionInput_StaticBool
   sortPriority=103
  }
  n24 = Multiply(A=NamedRerouteUsage(), B=@n10)
  n25 = Add(A=Add(A=Multiply(A=@n13, B=@n14.R), B=Multiply(A=@n15, B=@n16)), B=Multiply(A=StaticSwitch(True=@n22, False=@n14.R, Value=@n23), B=@n24)) {constB=0}
  n26 = FunctionInput(Preview=WorldPosition().XYZ) {
   bUsePreviewValueAsDefault=true
   description=World position must be subtracted from the World Position Offset output before being inserted into the World Position Offset pin. This input allows advanced users to override the world position value that's subtracted from the output.
   inputName=World Position to Negate
   sortPriority=107
  }
  n27 = Add(A=@n11, B=Add(A=Subtract(A=@n25, B=@n26), B=NamedRerouteUsage())) {desc="+ particle position "}
  n28 = Subtract(A=CameraPositionWS(), B=@n27)
  n29 = Multiply(A=Multiply(A=Normalize(VectorInput=@n28), B=NamedRerouteUsage())) {constB=0.5}
  n30 = FunctionInput(Preview=StaticBool()) {
   bUsePreviewValueAsDefault=true
   description=This feature pulls the cards closer to the camera. It along with PDO can better integrate the psuedo capsules into the scene. This option defaults to false.
   inputName=Use Camera Offset
   inputType=FunctionInput_StaticBool
   sortPriority=102
  }
  FunctionOutput(None=StaticSwitch(True=Add(A=@n29, B=@n27), False=@n27, Value=@n30)) {
   description=Note that the output of this material function subtracts world position from the results. This produces the correct results when plugged directly into world position offset.
   outputName=World Position Offset
   sortPriority=3
  }
  n31 = OneMinus(None=SphereMask(
    A=MaterialFunctionCall() {materialFunction=&LinearGradient}.VGradient
    B=Constant() {r=0.5}
    Radius=Constant() {desc=Radius of the engine's sphere texture, r=0.3794}
   ) {attenuationRadius=0.5, hardnessPercent=0})
  n32 = Multiply(A=@n31, B=@n31)
  n33 = OneMinus(None=Multiply(A=@n32, B=@n32))
  n34 = NamedRerouteUsage()
  n35 = NamedRerouteUsage()
  n36 = Multiply(
   A=NamedRerouteDeclaration(None=DotProduct(A=@n34, B=@n35)) {name=Particle Alignment With Camera, nodeColor=<rgba 0 1 0.9297 1>}
   B=Constant() {desc=Amount of uv offset that's appropriate for the given textures, r=0.3683}
  )
  n37 = StaticSwitch(True=Add(A=Multiply(A=@n33, B=@n36)) {constB=0.5}, False=Constant() {r=0.5}, Value=@n23)
  n38 = MaterialFunctionCall() {materialFunction=&LinearGradient}
  n39 = FunctionInput() {bUsePreviewValueAsDefault=true, inputName=Cap Size Multiplier, inputType=FunctionInput_Scalar, sortPriority=101, previewValue=<0.75 0 0 1>}
  n40 = Multiply(A=NamedRerouteUsage(), B=@n39) {constB=0.75}
  n41 = CameraPositionWS()
  n42 = Add(A=NamedRerouteUsage(), B=ParticlePositionWS())
  n43 = DotProduct(A=NamedRerouteUsage(), B=Normalize(VectorInput=Subtract(A=@n41, B=@n42)))
  n44 = WorldPosition()
  n45 = DotProduct(A=Normalize(VectorInput=Subtract(A=@n44.XYZ, B=@n42)), B=NamedRerouteUsage())
  n46 = FunctionOutput(None=Saturate(None=Multiply(A=@n43, B=@n45))) {description="The forward most edge returns a 1 whereas the receding edge returns a 0. ", outputName=Front Face Mask}
  n47 = LinearInterpolate(A=Divide(A=Constant() {r=55}) {constB=10}, B=Divide(A=Constant() {r=4}) {constB=10}, Alpha=@n46)
  n48 = Subtract(A=@n18, B=Subtract(A=Constant() {r=1}, B=@n47))
  n49 = Divide(B=@n47)
  n50 = Power(Base=Clamp(None=Multiply(A=@n48, B=@n49)), Exp=LinearInterpolate(A=Constant() {r=16}, B=Constant() {r=9}, Alpha=@n46))
  n51 = Clamp(None=Multiply(A=Divide(A=@n10, B=@n40), B=OneMinus(None=@n50))) {clampMode=CMODE_ClampMin, minDefault=1}
  n52 = FunctionOutput(None=AppendVector(
    A=MaterialFunctionCall!(Center Point=@n37, 0-1 Gradient=@n38.UGradient, Scale Factor=@n51) {materialFunction=/Engine/Functions/Engine_MaterialFunctions02/Gradients/StretchGradient}.Resulting Gradient
    B=@n38.VGradient
   )) {description=These uvs are distorted to better represent a 3D capsule., outputName=Texture Coordinates, sortPriority=10}
  n53 = NamedRerouteUsage()
  FunctionOutput(None=MaterialFunctionCall(
    VectorToTransform=TextureSample(UVs=@n52) {samplerType=SAMPLERTYPE_Normal, texture=/Engine/Functions/Engine_MaterialFunctions02/ExampleContent/Textures/SphereRenderNormalsMap}.RGB
    BasisX=@n12
    BasisY=NamedRerouteUsage()
    BasisZ=Multiply(A=@n53) {constB=-1}
   ) {materialFunction=&Transform3x3Matrix}.Transformed Vector) {
   description="These normals are in world space. Ensure that tangent space normals are unchecked in your material's details panel. "
   outputName=World Space Normals
   sortPriority=2
  }
  FunctionOutput(None=TextureSample(UVs=@n52) {texture=/Engine/Functions/Engine_MaterialFunctions02/ExampleContent/Textures/SphereRenderMask}.R) {outputName=Opacity Mask, sortPriority=1}
  n54 = TextureSample(UVs=@n52) {samplerType=SAMPLERTYPE_LinearColor, texture=/Engine/Functions/Engine_MaterialFunctions02/ExampleContent/Textures/SphereRenderHeightMap}
  FunctionOutput(None=Multiply(A=OneMinus(None=@n54.R), B=NamedRerouteUsage())) {outputName=Pixel Depth Offset, sortPriority=9}
  FunctionOutput(None=@n9) {description=This output can be used to modulate your particles emissive value etc., outputName=Scale Factor, sortPriority=15}
  n55 = FunctionInput(Preview=MaterialFunctionCall(Vector=ParticleDirection(), Default=Constant3Vector() {constant=<rgba 0 0 1 0>}) {materialFunction=&SafeNormalize}.Result) {
   bUsePreviewValueAsDefault=true
   description="Note: the particle will disappear if the system is fed a particle direction of  (0,0,0)."
   inputName=Normalized Particle Direction
  }
  NamedRerouteDeclaration(None=@n55) {name=Normalized Particle Direction, nodeColor=<rgba 1 0 0.3281 1>}
  n56 = NamedRerouteDeclaration(None=FunctionInput(Preview=ParticlePositionWS()) {bUsePreviewValueAsDefault=true, inputName=Particle Position, sortPriority=-1}) {name=Particle Position, nodeColor=<rgba 0.5703 0 1 1>}
  n57 = CameraPositionWS()
  NamedRerouteDeclaration(None=Normalize(VectorInput=Subtract(A=@n56, B=@n57) {desc=vector to camera from particle})) {name=Vector to Camera From Particle, nodeColor=<rgba 0 1 0.9531 1>}
  n58 = FunctionInput(Preview=ComponentMask(None=@n2) {r=true}) {bUsePreviewValueAsDefault=true, inputName=Capsule Diameter, inputType=FunctionInput_Scalar, previewValue=<10 0 0 1>}
  NamedRerouteDeclaration(None=@n58) {name=Capsule Diameter, nodeColor=<rgba 0.0547 1 0 1>}
  n59 = NamedRerouteUsage()
  NamedRerouteDeclaration(None=Normalize(VectorInput=CrossProduct(A=@n24, B=@n59))) {name=Up Vector, nodeColor=<rgba 0.3125 0 1 1>}
  NamedRerouteDeclaration(None=@n11) {name=Pivot offset, nodeColor=<rgba 0.2188 0 1 1>}
  NamedRerouteDeclaration(None=@n23) {name=Use Advanced Features Bool, nodeColor=<rgba 0 1 0.2266 1>}
  Subtract(A=@n45)
 ]
}
