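# Material function for World Position Offset that stretches particle sprites into velocity-aligned teardrops when moving across the view.
# A squared speed ramp, scaled by how perpendicular motion is to the view, elongates and narrows one end.
# Sprite corners are rebuilt on view-space velocity and perpendicular axes, using only particle size X.
# Per-pixel projection onto those axes remaps UVs to sample sphere textures for opacity mask and world-space normals; the height-map branch is unconnected.
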MaterialFunction /Engine/Functions/Engine_MaterialFunctions02/WorldPositionOffset/Sprite_TearDrop
graph {
 properties {
  bExposeToLibrary=true
  description=Stretch a particle out along the velocity vector when it's velocity is perpendicular to the view vector.
  libraryCategoriesText [World Position Offset]
 }
 nodes [
  FunctionInput() {bUsePreviewValueAsDefault=true, inputName=------ Particle Defaults Below ------, inputType=FunctionInput_StaticBool, sortPriority=10}
  n1 = TextureCoordinate()
  n2 = MaterialFunctionCall(Float2=ConstantBiasScale(None=@n1) {bias=-0.5, scale=1}) {materialFunction=&BreakOutFloat2Components}
  n3 = Constant3Vector() {desc=mask Z, constant=<rgba 1 1 0 0>}
  n4 = FunctionInput(Preview=ParticlePositionWS()) {bUsePreviewValueAsDefault=true, inputName=Particle Position, sortPriority=11}
  n5 = Normalize(VectorInput=Subtract(A=@n4, B=CameraPositionWS()) {desc=Vector toward the particle})
  n6 = MaterialFunctionCall(
   Vector=FunctionInput(Preview=ParticleDirection() {desc=x}) {bUsePreviewValueAsDefault=true, inputName=Particle Direction, sortPriority=11}
   Default=Constant3Vector() {constant=<rgba 0 0 1 0>}
  ) {materialFunction=&SafeNormalize}
  n7 = Multiply(
   A=@n3
   B=Transform(None=CrossProduct(A=@n6.Result, B=@n5) {desc=y/ screen right}) {transformSourceType=TRANSFORMSOURCE_World, transformType=TRANSFORM_View}
  )
  n8 = Multiply(A=Transform(None=@n6.Result) {transformSourceType=TRANSFORMSOURCE_World, transformType=TRANSFORM_View}, B=@n3)
  n9 = Normalize(VectorInput=CrossProduct(A=@n7, B=CrossProduct(A=@n8, B=@n7)))
  n10 = Constant2Vector() {g=1, r=1}
  n11 = FunctionInput() {
   bUsePreviewValueAsDefault=true
   description=The tail end of the sprite will shrink by this percentage when it's traveling perpendicular to the view and traveling at the maximum speed.
   inputName=Min Tear Drop Width
   inputType=FunctionInput_Scalar
  }
  n12 = FunctionInput() {
   bUsePreviewValueAsDefault=true
   description=The tail end of the sprite will elongate by this percentage when it's traveling perpendicular to the view and traveling at the maximum speed.
   inputName=Max Stretch Factor
   inputType=FunctionInput_Scalar
   previewValue=<1.6 0 0 1>
  }
  n13 = OneMinus(None=Abs(None=DotProduct(A=@n6.Result, B=@n5)))
  n14 = FunctionInput(Preview=ParticleSpeed()) {bUsePreviewValueAsDefault=true, inputName=Particle Speed, inputType=FunctionInput_Scalar, sortPriority=11}
  n15 = FunctionInput() {
   bUsePreviewValueAsDefault=true
   description=The particle must go this fast before beginning to warp along the velocity vector.
   inputName=Minimum Speed
   inputType=FunctionInput_Scalar
   sortPriority=1
   previewValue=<10 0 0 1>
  }
  n16 = FunctionInput() {
   bUsePreviewValueAsDefault=true
   description="After reaching the minimum speed, the particle will start to warp until this speed is reached (in addition to the minimum speed)."
   inputName=Max Speed
   inputType=FunctionInput_Scalar
   sortPriority=2
   previewValue=<30 0 0 1>
  }
  n17 = Clamp(None=Divide(A=Subtract(A=@n14, B=@n15), B=@n16))
  n18 = Multiply(A=@n17, B=@n17)
  n19 = LinearInterpolate(
   A=LinearInterpolate(A=@n10, B=AppendVector(A=@n11, B=@n12), Alpha=Multiply(A=@n13, B=@n18)) {desc=Width Factor}
   B=@n10
   Alpha=MaterialFunctionCall(Float2=@n1) {materialFunction=&BreakOutFloat2Components}.R
  )
  n20 = ComponentMask(None=@n19) {g=true}
  n21 = Multiply(A=@n2.R, B=Multiply(A=@n9, B=@n20))
  n22 = Normalize(VectorInput=@n7)
  n23 = ComponentMask(None=@n19) {r=true}
  n24 = Multiply(A=@n2.G, B=Multiply(A=@n22, B=@n23))
  n25 = FunctionInput(Preview=ParticleSize()) {bUsePreviewValueAsDefault=true, inputName=Particle Size, inputType=FunctionInput_Vector2, sortPriority=11}
  n26 = MaterialFunctionCall(Float2=@n25) {materialFunction=&BreakOutFloat2Components}
  n27 = Multiply(A=Transform(None=Add(A=@n21, B=@n24)) {transformSourceType=TRANSFORMSOURCE_View}, B=@n26.R)
  n28 = WorldPosition() {worldPositionShaderOffset=WPT_ExcludeAllShaderOffsets}
  FunctionOutput(None=Add(A=Subtract(A=@n27, B=@n28.XYZ), B=@n4)) {outputName=World Position Offset, sortPriority=4}
  n29 = Transform(None=Subtract(A=WorldPosition().XYZ, B=@n4)) {transformSourceType=TRANSFORMSOURCE_World, transformType=TRANSFORM_View}
  n30 = MaterialFunctionCall(Float2=@n25) {materialFunction=&BreakOutFloat2Components}
  n31 = Divide(A=DotProduct(A=@n9, B=@n29), B=@n30.R)
  n32 = Divide(A=DotProduct(A=@n22, B=@n29), B=@n30.R)
  n33 = Add(A=AppendVector(A=Divide(A=@n31, B=@n20), B=Divide(A=@n32, B=@n23))) {constB=0.5}
  n34 = FunctionOutput(None=Add(
    A=Saturate(None=@n33)
    B=FunctionInput() {bUsePreviewValueAsDefault=true, inputName=UV Add, inputType=FunctionInput_Vector2, sortPriority=12}
   )) {outputName=Texture Coordinates, sortPriority=55}
  n35 = TextureSample(UVs=@n34) {samplerType=SAMPLERTYPE_Normal, texture=/Engine/Functions/Engine_MaterialFunctions02/ExampleContent/Textures/SphereRenderNormalsMap}
  n36 = Multiply(A=@n35.R, B=@n9)
  n37 = Multiply(A=@n35.G, B=@n22)
  n38 = Constant3Vector() {constant=<rgba 0 0 -1 0>}
  n39 = Transform(None=Add(A=Add(A=@n36, B=@n37), B=Multiply(A=@n35.B, B=@n38))) {transformSourceType=TRANSFORMSOURCE_View}
  FunctionOutput(None=@n39) {outputName=World Space Normals, sortPriority=3}
  FunctionOutput(None=TextureSample(UVs=@n34) {texture=/Engine/Functions/Engine_MaterialFunctions02/ExampleContent/Textures/SphereRenderMask}.R) {outputName=Opacity Mask, sortPriority=2}
  MaterialFunctionCall(Vector=@n39, Default=Constant3Vector() {constant=<rgba 0 0 1 0>}) {materialFunction=&SafeNormalize}
  n40 = TextureSample(UVs=@n34) {samplerType=SAMPLERTYPE_LinearColor, texture=/Engine/Functions/Engine_MaterialFunctions02/ExampleContent/Textures/SphereRenderHeightMap}
  Multiply(A=@n26.R, B=OneMinus(None=@n40.R))
  ParticlePositionWS()
 ]
}
