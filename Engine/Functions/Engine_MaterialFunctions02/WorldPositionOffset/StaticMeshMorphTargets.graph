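# Material function that unpacks morph target data baked into static meshes by the 3ds Max Morph Packer script, for world position offset animation.
# UV channels 1 to 3 are appended into vectors and transformed from instance space for two morph offsets.
# UV1 plus UV2 red serves as either the second morph or a world-space pivot position, depending on script options.
# Vertex color is expanded to signed range as tangent-space normals for the first morph.
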MaterialFunction /Engine/Functions/Engine_MaterialFunctions02/WorldPositionOffset/StaticMeshMorphTargets
graph {
 properties {
  bExposeToLibrary=true
  description=Unpacks Morph Data From the 3ds Max Morph Packer Script
  libraryCategoriesText [Misc, World Position Offset]
 }
 nodes [
  n1 = ComponentMask(None=TextureCoordinate() {coordinateIndex=2}) {g=true}
  n2 = TextureCoordinate() {coordinateIndex=3}
  FunctionOutput(None=Transform(None=AppendVector(A=@n1, B=@n2)) {transformSourceType=TRANSFORMSOURCE_Instance}) {outputName=Morph Target 1 WPO, sortPriority=2}
  n3 = TextureCoordinate() {coordinateIndex=1}
  n4 = ComponentMask(None=TextureCoordinate() {coordinateIndex=2}) {r=true}
  FunctionOutput(None=TransformPosition(None=AppendVector(A=@n3, B=@n4)) {transformSourceType=TRANSFORMPOSSOURCE_Instance, transformType=TRANSFORMPOSSOURCE_World}) {
   description=Returns the pivot point information for each element in world space if that option was chosen in the script at the time the model was processed. The output of this pin is in world space.
   outputName=Pivot Position WS
   sortPriority=10
  }
  n5 = TextureCoordinate() {coordinateIndex=1}
  n6 = ComponentMask(None=TextureCoordinate() {coordinateIndex=2}) {r=true}
  FunctionOutput(None=Transform(None=AppendVector(A=@n5, B=@n6)) {transformSourceType=TRANSFORMSOURCE_Instance}) {outputName=Morph Target 2 WPO, sortPriority=3}
  n7 = VertexColor()
  FunctionOutput(None=Transform(None=ConstantBiasScale(None=@n7) {bias=-0.5, scale=2}) {transformSourceType=TRANSFORMSOURCE_Instance, transformType=TRANSFORM_Tangent}) {outputName=Morph Target 1 Normals}
  FunctionOutput() {outputName=---------------, sortPriority=1}
 ]
}
