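# Material function for World Position Offset that visualizes a mesh's UV coordinates in 3D by moving vertices into their UV layout.
# The input UV, defaulting to a texture coordinate, gets its V flipped, is centered on zero and is scaled by twice the object radius.
# That planar position is transformed from view space, placed at the actor position, then has world position subtracted to yield the offset.
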MaterialFunction /Engine/Functions/Engine_MaterialFunctions02/WorldPositionOffset/UVLayoutToWorldSpacePosition
graph {
 properties {
  bExposeToLibrary=true
  description="Use in the world position offset input to visual uv coordinates in 3D. "
  libraryCategoriesText [Misc, World Position Offset]
 }
 nodes [
  n1 = ObjectRadius()
  n2 = Add(A=@n1, B=@n1)
  n3 = FunctionInput(Preview=TextureCoordinate()) {
   bUsePreviewValueAsDefault=true
   description=Insert a uv coordinate to view in 3d space.
   inputName=Texture Coordinate
   inputType=FunctionInput_Vector2
  }
  n4 = MaterialFunctionCall(Float2=@n3) {materialFunction=&BreakOutFloat2Components}
  n5 = Subtract(A=AppendVector(A=@n4.R, B=OneMinus(None=@n4.G)), B=Constant2Vector() {r=0.5}) {constB=0.5}
  n6 = Transform(None=AppendVector(A=Multiply(A=@n2, B=@n5), B=Constant())) {transformSourceType=TRANSFORMSOURCE_View}
  n7 = ActorPositionWS()
  FunctionOutput(None=Subtract(A=Add(A=@n6, B=@n7), B=WorldPosition().XYZ)) {outputName=World Positon}
 ]
}
